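# Unlit, translucent, two-sided Niagara ribbon material for the BalletAssassin pickaxe melee trail.
# Panning-noise soft and hard leading edges, a sphere-mask hard line and a stepped, scrolling UV1 pattern layer the colors, tinted by particle color.
# Opacity sums those masks with two panning motion-line layers, then applies four UV edge fades and particle alpha.
# A dynamic parameter can offset soft-edge UVs, off by default.
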Material /BRCosmetics/Weapons/FORT_Melee/Pickaxe_BalletAssassin/FX/M_BalletAssassin_MeleeTrail
graph {
 properties {
  bAutomaticallySetUsageInEditor=false
  bUsedWithNiagaraRibbons=true
  bUsedWithStaticMesh=false
  blendMode=BLEND_Translucent
  pixelDepthOffsetMode=PDOM_Legacy
  shadingModel=MSM_Unlit
  twoSided=true
  usedShadingModels=Unlit
 }
 nodes [
  n1 = DynamicParameter() {defaultValue=<rgba 0 0 1 1>, paramNames [X Offset, Y Offset, Param3, Param4]}
  n2 = TextureCoordinate()
  n3 = VectorParameter() {group=Soft Leading Edge, parameterName=Soft Leading Edge UV Scale, sortPriority=6, defaultValue=<rgba 0.2 1 0 0>}
  n4 = VectorParameter() {group=Soft Leading Edge, parameterName=Soft Leading Edge Panner Speed, sortPriority=7, defaultValue=<rgba -0.65 0 0 0>}
  n5 = Panner(Coordinate=Multiply(A=@n2, B=ComponentMask(None=@n3.RGB) {g=true, r=true}), Speed=ComponentMask(None=@n4.RGB) {g=true, r=true}) {bFractionalPart=true}
  n6 = Add(A=@n5, B=AppendVector(A=@n1.X Offset, B=@n1.X Offset))
  n7 = ScalarParameter() {group=Soft Leading Edge, parameterName=Use_Dynamic_UV-Offsets, sortPriority=0}
  n8 = TextureSampleParameter2D(UVs=Add(A=NamedRerouteUsage(), B=LinearInterpolate(A=@n5, B=@n6, Alpha=@n7))) {
   group=Soft Leading Edge
   parameterName=Soft Leading Edge Texture
   samplerSource=SSM_Wrap_WorldGroupSettings
   sortPriority=4
   texture=&T_NoisePack_01
  }
  n9 = Add(
   A=ChannelMaskParameter(None=@n8.RGBA) {group=Soft Leading Edge, parameterName=Soft Leading Edge Texture Mask, sortPriority=5}.R
   B=ScalarParameter() {defaultValue=-0.5, group=Soft Leading Edge, parameterName=Bias, sortPriority=3}
  )
  n10 = ScalarParameter() {defaultValue=0.2, group=Soft Leading Edge, parameterName=Scale, sortPriority=2}
  n11 = ComponentMask(None=TextureCoordinate()) {r=true}
  n12 = Power(
   Base=Add(A=Multiply(A=@n9, B=@n10), B=OneMinus(None=@n11))
   Exp=ScalarParameter() {defaultValue=8, group=Soft Leading Edge, parameterName=Soft Leading Edge Contrast, sortPriority=1}
  ) {constExponent=6}
  n13 = VectorParameter() {group=Hard Leading Edge, parameterName=Hard Leading Edge UV Scale A, sortPriority=4, defaultValue=<rgba 1.2 1.5 0 0>}
  n14 = MaterialFunctionCall(Float2=Multiply(A=@n2, B=ComponentMask(None=@n13.RGB) {g=true, r=true})) {materialFunction=&BreakOutFloat2Components}
  n15 = MaterialFunctionCall(
   X=OneMinus(None=@n14.G)
   Y=Add(A=@n14.R, B=ScalarParameter() {defaultValue=-0.5, group=Hard Leading Edge, parameterName=Hard Leading Edge Bias A, sortPriority=5}) {constA=0.25, constB=-0.3}
  ) {materialFunction=&MakeFloat2}
  n16 = VectorParameter() {group=Hard Leading Edge, parameterName=Hard Leading Edge UV Scale B, sortPriority=5, defaultValue=<rgba 12 3 0 0>}
  n17 = MaterialFunctionCall(Float2=Multiply(A=@n2, B=ComponentMask(None=@n16.RGB) {g=true, r=true})) {materialFunction=&BreakOutFloat2Components}
  n18 = MaterialFunctionCall(
   X=OneMinus(None=@n17.G)
   Y=Add(A=ScalarParameter() {defaultValue=-1, group=Hard Leading Edge, parameterName=Hard Leading Edge Bias B, sortPriority=6}, B=@n17.R) {constA=0.25}
  ) {materialFunction=&MakeFloat2}
  n19 = ComponentMask(None=TextureCoordinate()) {g=true}
  n20 = VectorParameter() {group=Hard Leading Edge, parameterName=Hard Edge Mask Panner Speed, sortPriority=7}
  n21 = TextureSampleParameter2D(UVs=Panner(Coordinate=LinearInterpolate(A=@n15.Result, B=@n18.Result, Alpha=@n19), Speed=ComponentMask(None=@n20.RGB) {g=true, r=true})) {
   group=Hard Leading Edge
   parameterName=Hard Leading Edge Texture
   samplerType=SAMPLERTYPE_LinearColor
   sortPriority=2
   texture=/Game/CharClasses/Soldier/Traits/AdvancedTactics/Abilities/GoingCommando/Textures/T_GoingCommando_RingNoise_01
  }
  n22 = OneMinus(None=@n19)
  n23 = Multiply(
   A=ChannelMaskParameter(None=@n21.RGBA) {group=Hard Leading Edge, parameterName=Hard Leading Edge Texture Mask, sortPriority=3}.R
   B=Multiply(A=@n22, B=@n22)
  )
  n24 = Saturate(None=Add(A=Multiply(A=@n12) {constB=0.8}, B=Multiply(A=@n23) {constB=0.6}) {constB=0})
  n25 = Constant() {r=0.005}
  n26 = ComponentMask(None=TextureCoordinate()) {g=true}
  n27 = ScalarParameter() {defaultValue=0.075, group=Hard Line, parameterName=Edge Radius A, sortPriority=1}
  n28 = ScalarParameter() {defaultValue=0.005, group=Hard Line, parameterName=Edge Radius B, sortPriority=2}
  n29 = ComponentMask(None=TextureCoordinate()) {r=true}
  n30 = ScalarParameter() {defaultValue=0.3, group=Hard Line, parameterName=Edge Hardness A, sortPriority=3}
  n31 = ScalarParameter() {group=Hard Line, parameterName=Edge Hardness B, sortPriority=4}
  n32 = Saturate(None=SphereMask(A=AppendVector(A=@n25, B=@n26), B=@n25, Radius=LinearInterpolate(A=@n27, B=@n28, Alpha=@n29), Hardness=LinearInterpolate(A=@n30, B=@n31)))
  n33 = LinearInterpolate(
   A=LinearInterpolate(
    A=VectorParameter() {parameterName=BaseColor, defaultValue=<rgba 0.9 0.9 0.9 1>}.RGB
    B=VectorParameter() {parameterName=LeadingEdge, defaultValue=<rgba 0.6 0.45 0.24 1>}.RGB
    Alpha=@n24
   )
   B=VectorParameter() {parameterName=TopColor, defaultValue=<rgba 1.5 1.1 0.3 1>}.RGB
   Alpha=@n32
  )
  n34 = VectorParameter() {parameterName=PatternColor, defaultValue=<rgba 0.18 0.3 0.9 1>}
  n35 = TextureSampleParameter2D(UVs=Panner(Coordinate=TextureCoordinate() {coordinateIndex=1}) {speedX=1}) {
   parameterName=Pattern
   samplerType=SAMPLERTYPE_Masks
   texture=&/T_BalletAssassin_Trail_Mask
  }
  n36 = OneMinus(None=Step(Y=@n35.G) {constX=0.1})
  n37 = ParticleColor()
  MaterialFunctionCall(Emissive=Multiply(A=LinearInterpolate(A=@n33, B=@n34.RGB, Alpha=@n36) {constA=1}, B=@n37.RGB) {constA=1}) {materialFunction=&MF_FX_EmissiveAdjustments}
  n38 = ScalarParameter() {defaultValue=0.2, group=Alpha, parameterName=Fine Motion Lines Opacity A, sortPriority=10}
  n39 = ScalarParameter() {defaultValue=0.3, group=Alpha, parameterName=Fine Motion Lines Opacity B, sortPriority=11}
  n40 = ScalarParameter() {defaultValue=1, group=Motion Lines, parameterName=Motion Lines A Intensity, sortPriority=3}
  n41 = VectorParameter() {group=Motion Lines, parameterName=Fine Motion Lines UV Scale A, sortPriority=5, defaultValue=<rgba 1 3.464 0 0>}
  n42 = VectorParameter() {group=Motion Lines, parameterName=Fine Motion Lines Panner Speed A, sortPriority=4, defaultValue=<rgba 0 0.976 0 0>}
  n43 = Panner(Coordinate=Multiply(A=@n2, B=ComponentMask(None=@n41.RGB) {g=true, r=true}), Speed=ComponentMask(None=@n42.RGB) {g=true, r=true})
  n44 = MaterialFunctionCall(XY=@n43) {materialFunction=&Swizzle}
  n45 = ScalarParameter() {defaultValue=1, group=Motion Lines, parameterName=Motion Lines Swizzle UVs, sortPriority=12}
  n46 = NamedRerouteUsage()
  n47 = TextureSampleParameter2D(UVs=Add(A=LinearInterpolate(A=@n43, B=@n44.YX, Alpha=@n45), B=@n46)) {
   group=Motion Lines
   parameterName=Motion Lines Texture
   samplerSource=SSM_Wrap_WorldGroupSettings
   sortPriority=6
   texture=&T_NoisePack_01
  }
  n48 = Power(
   Base=ChannelMaskParameter(None=@n47.RGBA) {group=Motion Lines, parameterName=Motion Lines Texture Mask A, sortPriority=1}.R
   Exp=ScalarParameter() {defaultValue=1.5, group=Motion Lines, parameterName=Motion Lines Texture A Contrast, sortPriority=2}
  ) {constExponent=1.5}
  n49 = VectorParameter() {group=Motion Lines, parameterName=Fine Motion Lines UV Scale B, sortPriority=11, defaultValue=<rgba 1 4.4362 0 0>}
  n50 = VectorParameter() {group=Motion Lines, parameterName=Fine Motion Lines Panner Speed B, sortPriority=10, defaultValue=<rgba 0 -0.7616 0 0>}
  n51 = Panner(Coordinate=Multiply(A=@n2, B=ComponentMask(None=@n49.RGB) {g=true, r=true}), Speed=ComponentMask(None=@n50.RGB) {g=true, r=true}) {speedY=-0.7616}
  n52 = MaterialFunctionCall(XY=@n51) {materialFunction=&Swizzle}
  n53 = TextureSampleParameter2D(UVs=Add(A=LinearInterpolate(A=@n51, B=@n52.YX, Alpha=@n45), B=@n46)) {
   group=Motion Lines
   parameterName=Motion Lines Texture
   samplerSource=SSM_Wrap_WorldGroupSettings
   sortPriority=6
   texture=&T_NoisePack_01
  }
  n54 = ChannelMaskParameter(None=@n53.RGBA) {group=Motion Lines, maskChannel=Green, parameterName=Motion Lines Texture Mask B, sortPriority=7, defaultValue=<rgba 0 1 0 0>}
  n55 = Power(Base=@n54.R, Exp=ScalarParameter() {defaultValue=2, group=Motion Lines, parameterName=Motion Lines Texture B Contrast, sortPriority=8})
  n56 = ScalarParameter() {defaultValue=1, group=Motion Lines, parameterName=Motion Lines B Intensity, sortPriority=9}
  n57 = Multiply(
   A=Add(A=Multiply(A=@n40, B=@n48), B=Multiply(A=@n55, B=@n56))
   B=ScalarParameter() {defaultValue=0.5, group=Motion Lines, parameterName=Combined Motion Lines Scale}
  )
  n58 = Add(A=Multiply(A=LinearInterpolate(A=@n38, B=@n39, Alpha=@n57) {constA=1}) {constB=0.5}, B=@n24)
  n59 = ComponentMask(None=TextureCoordinate()) {g=true}
  n60 = ScalarParameter() {group="Fade Top && Bottom Length Edges", parameterName=Top Length Fade Bias, sortPriority=2}
  n61 = Saturate(None=Multiply(A=Add(A=@n59, B=@n60), B=ScalarParameter() {defaultValue=16, group="Fade Top && Bottom Length Edges", parameterName=Top Length Fade Scale}) {constB=3})
  n62 = ScalarParameter() {defaultValue=1, group=Alpha, parameterName=Fade Top Length Edge, sortPriority=7}
  n63 = OneMinus(None=@n59)
  n64 = ScalarParameter() {group="Fade Top && Bottom Length Edges", parameterName=Bottom Length Fade Bias, sortPriority=3}
  n65 = Saturate(None=Multiply(
    A=Add(A=@n63, B=@n64)
    B=ScalarParameter() {defaultValue=4, group="Fade Top && Bottom Length Edges", parameterName=Bottom Length Fade Scale, sortPriority=1}
   ) {constB=2})
  n66 = ScalarParameter() {defaultValue=1, group=Alpha, parameterName=Fade Bottom Length Edge, sortPriority=8}
  n67 = Multiply(
   A=Add(A=@n36, B=Add(A=@n32, B=@n58))
   B=Multiply(A=LinearInterpolate(B=@n61, Alpha=@n62) {constA=1}, B=LinearInterpolate(B=@n65, Alpha=@n66) {constA=1})
  )
  n68 = ComponentMask(None=TextureCoordinate()) {r=true}
  n69 = Multiply(
   A=Add(A=@n68, B=ScalarParameter() {group="Fade Leading && Trailing Edges", parameterName=Leading Edge Fade Bias, sortPriority=2})
   B=ScalarParameter() {defaultValue=4, group="Fade Leading && Trailing Edges", parameterName=Leading Edge Fade Scale}
  ) {constB=3}
  n70 = Multiply(
   A=Add(A=OneMinus(None=@n68), B=ScalarParameter() {group="Fade Leading && Trailing Edges", parameterName=Trailing Edge Fade Bias, sortPriority=3})
   B=ScalarParameter() {defaultValue=1, group="Fade Leading && Trailing Edges", parameterName=Trailing Edge Fade Scale, sortPriority=1}
  ) {constB=2}
  n71 = Multiply(
   A=LinearInterpolate(B=Saturate(None=@n69), Alpha=ScalarParameter() {defaultValue=1, group=Alpha, parameterName=Fade Leading Edge, sortPriority=4}) {constA=1}
   B=LinearInterpolate(B=Saturate(None=@n70), Alpha=ScalarParameter() {defaultValue=1, group=Alpha, parameterName=Fade Trailing Edge, sortPriority=5}) {constA=1}
  )
  Multiply(A=@n37.A, B=Saturate(None=Multiply(A=@n67, B=@n71)))
 ]
}
